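# Conversation database for the SurgeRaven NPC, a revenge-seeking former Black Mamba hunting Bill.
# An entry-tagged greeting picks contextual lines using first-meeting, NPC conversation-count and player-outfit requirements, with priority curve rows for first meeting and outfit reactions.
# It branches to link nodes for shared quests, basic services, random services and a Season 39 PerfectGift conversation.
# An "Ask About..." choice offers a linear Deadly Viper Assassination Squad backstory that returns to the start.
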ConversationDatabase /NPCLibrary/NPCs/SurgeRaven/NPCLibrary_Conversation_SurgeRaven
entry_points [{tag=Conversation.Entry.SurgeRaven, destinations [#1]}]
nodes [
 ConversationEntryPointNode_7 {
  id=#1
  class=ConversationEntryPointNode
  outputs [#2]
  properties {entryTag {tagName=Conversation.Entry.SurgeRaven}}
 }
 Speech_C_4 {
  id=#2
  class=Speech_C
  outputs [
   #3
   #4
   #5
   #6
   #7
  ]
  properties {
   nodeName=""
   generalConfig {
    defaultMessage=Where's Bill?
    contextualMessages [
     {
      message=Where's Bill?
      contextRequirements [{requirement {class=FortControllerRequirement_IsFirstEverConversationWithNPC}}]
      priority {curve {rowName=Tandem.Speech.Priority.FirstEverInteraction}}
     }
     {
      message=They called me the Black Mamba... a long time ago.
      contextRequirements [{requirement {class=FortControllerRequirement_CheckNPCMatchConversationCount, properties {nPCUniqueIdTag {tagName=AISpawnerData.Type.Tandem.SurgeRaven}}}}]
      priority {curve {curveTable=null, rowName=None}}
     }
     {
      message=Revenge is a dish best served cold.
      contextRequirements [{requirement {class=FortControllerRequirement_CheckNPCMatchConversationCount, properties {nPCUniqueIdTag {tagName=AISpawnerData.Type.Tandem.SurgeRaven}}}}]
      priority {curve {curveTable=null, rowName=None}}
     }
     {
      message=I have unfinished business.
      contextRequirements [{requirement {class=FortControllerRequirement_CheckNPCMatchConversationCount, properties {nPCUniqueIdTag {tagName=AISpawnerData.Type.Tandem.SurgeRaven}}}}]
      priority {curve {curveTable=null, rowName=None}}
     }
     {
      message="Sometimes all you have left is revenge. "
      contextRequirements [{requirement {class=FortControllerRequirement_CheckNPCMatchConversationCount, properties {nPCUniqueIdTag {tagName=AISpawnerData.Type.Tandem.SurgeRaven}}}}]
      priority {curve {curveTable=null, rowName=None}}
     }
     {
      message="What you lack in age, Gogo, you make up for in madness."
      contextRequirements [{requirement {class=FortControllerRequirement_HasCID, properties {allowedCIds [/BRCosmetics/Athena/Items/Cosmetics/Characters/Character_JoltMosaic]}}}]
      priority {curve {rowName=Tandem.Speech.Priority.OutfitReaction}}
     }
     {
      message="I understand your desire for revenge, more than most."
      contextRequirements [{requirement {
         class=FortControllerRequirement_HasCID
         properties {allowedCIds [/BRCosmetics/Athena/Items/Cosmetics/Characters/Character_WaltzScout, /BRCosmetics/Athena/Items/Cosmetics/Characters/Character_WaltzScout_V2]}
        }}]
      priority {curve {rowName=Tandem.Speech.Priority.OutfitReaction}}
     }
    ]
   }
  }
 }
 ConversationLinkNode_2 {
  id=#7
  class=ConversationLinkNode
  properties {remoteEntryTag {tagName=Conversation.Entry.Season39.WF.PerfectGift}}
 }
 Speech_C_2 {
  id=#6
  class=Speech_C
  outputs [#8, #9]
  properties {nodeName="", softTaskIcon=&T_UI_NPC_question, generalConfig {defaultMessage=What do you need to know?}}
  sub_nodes [ConversationChoiceNode_11 {class=ConversationChoiceNode, properties {defaultChoiceDisplayText=Ask About...}}]
 }
 Back_C_2 {
  id=#9
  class=Back_C
  sub_nodes [ConversationChoiceNode_40 {class=ConversationChoiceNode, properties {defaultChoiceDisplayText=BACK}}]
 }
 Speech_C_1 {
  id=#8
  class=Speech_C
  outputs [#10]
  properties {
   nodeName=""
   softTaskIcon=&T_UI_NPC_question
   generalConfig {defaultMessage="I used to be one of them. Then, at Bill's word, they betrayed me."}
  }
  sub_nodes [ConversationChoiceNode_6 {class=ConversationChoiceNode, properties {defaultChoiceDisplayText=Deadly Viper Assassination Squad}}]
 }
 Speech_C_6 {
  id=#10
  class=Speech_C
  outputs [#11]
  properties {
   nodeName=""
   softTaskIcon=&T_UI_NPC_question
   generalConfig {defaultMessage=I was the Black Mamba. I tried to leave. Bill... refused.}
  }
  sub_nodes [ConversationChoiceNode_8 {class=ConversationChoiceNode, properties {defaultChoiceDisplayText=Continue...}}]
 }
 Speech_C_8 {
  id=#11
  class=Speech_C
  outputs [#12]
  properties {
   nodeName=""
   softTaskIcon=&T_UI_NPC_question
   generalConfig {defaultMessage="But in the end, I'll have my revenge on all of them. Every single one."}
  }
  sub_nodes [ConversationChoiceNode_9 {class=ConversationChoiceNode, properties {defaultChoiceDisplayText=Continue...}}]
 }
 Speech_C_9 {
  id=#12
  class=Speech_C
  outputs [#13]
  properties {nodeName="", softTaskIcon=&T_UI_NPC_question, generalConfig {defaultMessage=Especially Bill.}}
  sub_nodes [ConversationChoiceNode_10 {class=ConversationChoiceNode, properties {defaultChoiceDisplayText=Continue...}}]
 }
 ReturnToStart_C_0 {
  id=#13
  class=ReturnToStart_C
  properties {softTaskIcon=&T_UI_NPC_Back}
  sub_nodes [ConversationChoiceNode_7 {class=ConversationChoiceNode, properties {defaultChoiceDisplayText=BACK}}]
 }
 ConversationLinkNode_3 {id=#5, class=ConversationLinkNode, properties {remoteEntryTag {tagName=Conversation.Entry.RandomServices}}}
 ConversationLinkNode_0 {id=#4, class=ConversationLinkNode, properties {remoteEntryTag {tagName=Conversation.Entry.BasicServices}}}
 ConversationLinkNode_1 {id=#3, class=ConversationLinkNode, properties {remoteEntryTag {tagName=Conversation.Entry.NPCLibrary.Quests}}}
]
